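# Niagara system for the "Bad" effect of the QuirkLuster player emote: one burst emitter drawing dark, short-lived, drifting sprites through two sprite renderers.
# Fixed facings on perpendicular axes and a second position set up crossed planes.
# Camera-vector dot products drive two color alphas, fading each by view angle.
# Curves animate overall alpha and dynamic material parameters over particle life.
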NiagaraSystem /BRCosmetics/Animation/Game/MainPlayer/Emotes/QuirkLuster/FX/NS_Emote_QuirkLuster_Bad
dependencies {usedDataInterfaces [&NiagaraDataInterfaceCamera], usedDynamicInputs [], usedModules [&SystemState]}
emitters {NE_Line01 {
  emitterUpdateScript [
   &EmitterState(Life Cycle Mode=::NewEnumerator1, Loop Behavior=::NewEnumerator1, Loop Duration=1, UseLoopDelay=false)
   &SpawnBurst_Instantaneous(Use Spawn Probability=false, UseLoopCountLimit=false)
  ]
  particleSpawnScript [
   &InitializeParticle(
    Color=<rgba 0 0.001 0.0003 2>
    Lifetime=2
    Mesh Scale Mode=::NewEnumerator3
    Mesh Uniform Scale=0.6
    Sprite Rotation Angle=-95
    Sprite Rotation Mode=::NewEnumerator0
    Sprite Size Mode=::NewEnumerator3
    Uniform Sprite Size=25
    UsePositionOffset=false
    Write Material Random=true
   ) {module=InitializeParticle001}
   &AddVelocity(Velocity=<-2 -1 0>)
   /SetVariables(Particles.Cam Offset=10) {isSetParameters=true}
   /SetVariables(Particles.Position2:NiagaraPosition=&AddVectorToPosition(Position=<10 -5 0>, Vector=<0 0 0>)) {isSetParameters=true}
  ]
  particleUpdateScript [
   &ParticleState()
   &ScaleColor(
    Scale Alpha=&FloatFromCurve(FloatCurve=DataInterface() {propertyValues {Curve {keys [<at 0 0>,<at 0.028 1>,<at 0.5 1>,<at 1 0 RCIM_Cubic>]}}})
    ScaleA=true
    ScaleRGB=true
   )
   &SolveForcesAndVelocity(Clamp Velocity=false, Limit Acceleration=false)
   &DynamicMaterialParameters(
    Index 0 Param 1=&FloatFromCurve(FloatCurve=DataInterface() {propertyValues {Curve {defaultValue=1, keys [{tangentMode=RCTM_User, time=0.0078, value=-0.8397}, {tangentMode=RCTM_User, time=1, value=1}]}}})
    Index 0 Param 2=1
    Index 0 Param 3=1
    Index 0 Param 4=&FloatFromCurve(FloatCurve=DataInterface() {propertyValues {Curve {keys [<at 0 1>,<at 0.2 1>,<at 1 0>]}}})
    Module.Param0WriteEnabled=true
    Module.Param1WriteEnabled=true
    Module.Param2WriteEnabled=true
    Module.Param3WriteEnabled=true
   )
   &SpriteFacingAndAlignment(Sprite Alignment=<0 0 1>, Sprite Facing=<0 1 0>, WriteAlignment=true, WriteFacing=true)
   /SetVariables(Particles.SecondAlignment:Vector3f=<0 0 1>) {isSetParameters=true}
   /SetVariables(Particles.SecondFacing:Vector3f=<1 0 0>) {isSetParameters=true}
   /SetVariables(Particles.Color1:LinearColor=&MakeLinearColorFromVectorAndFloat(
     "Vector (RGB)"=&MakeVectorFromLinearColorRGB(Color=<rgba 0 0.001 0.0003 2>)
     "Float (Alpha)"=&RemapRange(
      Input Value To Remap=&Abs_Float(Float:NiagaraFloat=&DotProduct(
        Vector A=&ReturnCameraVector(Camera Vectors=::NewEnumerator4, Coordinate Space=::Simulation, Origin Position=$Particles.Position2)
        Vector B=<0 1 0>
       ))
      Input Min=0.3
      Input Max=0.6
      Output Min=1
      Output Max=0
     )
    )) {isSetParameters=true}
   /SetVariables(Particles.Color2=&MakeLinearColorFromVectorAndFloat(
     "Vector (RGB)"=&MakeVectorFromLinearColorRGB(Color=<rgba 0 0.001 0.0003 2>)
     "Float (Alpha)"=&RemapRange(
      Input Value To Remap=&Abs_Float(Float:NiagaraFloat=&DotProduct(
        Vector A=&ReturnCameraVector(Camera Vectors=::NewEnumerator4, Coordinate Space=::Simulation, Origin Position=$Particles.Position)
        Vector B=<1 0 0>
       ))
      Input Min=0.5
      Input Max=0.9
      Output Min=1
      Output Max=0
     )
    )) {isSetParameters=true}
  ]
  rendererClasses [&NiagaraSpriteRendererProperties]
  renderers [
   {rendererClass=&NiagaraSpriteRendererProperties}
   {rendererClass=&NiagaraSpriteRendererProperties, rendererIndex=1}
  ]
 }}
summary {
 systemName=NS_Emote_QuirkLuster_Bad
 emitters [{emitterName=NE_Line01, rendererClasses [&NiagaraSpriteRendererProperties]}]
 userVariables [
  User.BaseCurveColor {
   type=NiagaraDataInterfaceColorCurve
   description=""
   defaultValue {
    struct=NiagaraExt_VariableValue_DataInterface
    value {
     dataInterface=&self:NiagaraDataInterfaceColorCurve_0
     dataInterfaceClass=&NiagaraDataInterfaceColorCurve
    }
   }
  }
  User.BaseCurveColorHueShift {
   type=NiagaraDataInterfaceCurve
   description=""
   defaultValue {
    struct=NiagaraExt_VariableValue_DataInterface
    value {
     dataInterface=&self:NiagaraDataInterfaceCurve_0
     dataInterfaceClass=&NiagaraDataInterfaceCurve
    }
   }
  }
  User.BaseCurveColorPick {type=NiagaraFloat, defaultValue=0, description=""}
 ]
}
