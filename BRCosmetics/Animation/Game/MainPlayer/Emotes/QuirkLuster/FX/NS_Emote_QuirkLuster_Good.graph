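# Niagara system for the "Good" variant of the QuirkLuster player emote: a warm glow with rising, fading star sprites.
# NE_Glow01 bursts two glow layers, while NE_Star01 rate-spawns sprites in a box, lifted by gravity and slowed by drag.
# Each emitter merges two looks: spawn group index drives MergeHelper switches for color, size and curves, and sets VisibilityTag alongside paired sprite renderers.
# BaseCurveColor user parameters are exposed but unreferenced in the listed modules.
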NiagaraSystem /BRCosmetics/Animation/Game/MainPlayer/Emotes/QuirkLuster/FX/NS_Emote_QuirkLuster_Good
dependencies {usedDataInterfaces [], usedDynamicInputs [], usedModules [&SystemState]}
emitters {
 NE_Glow01 {
  emitterUpdateScript [
   &EmitterState(Life Cycle Mode=::NewEnumerator1, Loop Behavior=::NewEnumerator1, Loop Duration=0.8, UseLoopDelay=false)
   &SpawnBurst_Instantaneous(Use Spawn Probability=false, UseLoopCountLimit=false)
   &SpawnBurst_Instantaneous(Spawn Group=1, Use Spawn Probability=false, UseLoopCountLimit=false) {module=SpawnBurst_Instantaneous001}
  ]
  particleSpawnScript [
   /SetVariables(Particles.VisibilityTag=&SpawnGroup()) {isSetParameters=true}
   &InitializeParticle(
    Color=&MergeHelper_ColorSwitch(
     0=<rgba 1 0.6943 0.5495 0.2174>
     1=<rgba 1 0.4345 0.2678 1>
     2=?
     3=?
     Index=&SpawnGroup()
     Num=::NewEnumerator0
    )
    Lifetime=1.5
    Mass Mode=::NewEnumerator3
    Sprite Rotation Mode=::NewEnumerator2
    Sprite Size Mode=::NewEnumerator3
    Uniform Sprite Size=&MergeHelper_FloatSwitch(0=50, 1=100, 2=?, 3=?, Index=&SpawnGroup(), Num=::NewEnumerator0)
    UsePositionOffset=false
    Write Material Random=true
   )
  ]
  particleUpdateScript [
   &ParticleState()
   &ScaleColor(
    Scale Alpha=&MergeHelper_FloatSwitch(
     0=&FloatFromCurve(
      FloatCurve=DataInterface() {propertyValues {Curve {keys [<at -0.0035 0.192>, <at 0.2 1>, {arriveTangent=-1.0036, time=0.8, value=1}, <at 1.0011 0.002 RCIM_Cubic>]}}}
      Scale Curve=0.3
     )
     1=&FloatFromCurve(FloatCurve=DataInterface() {propertyValues {Curve {keys [<at -0.0035 0.192>, <at 0.2 1>, {arriveTangent=-1.0036, time=0.8, value=1}, <at 1.0011 0.002 RCIM_Cubic>]}}})
     2=?
     3=?
     Index=&SpawnGroup()
     Num=::NewEnumerator0
    )
    ScaleA=true
    ScaleRGB=true
   )
   &SolveForcesAndVelocity(Clamp Velocity=false, Limit Acceleration=false)
  ]
  rendererClasses [&NiagaraSpriteRendererProperties]
  renderers [
   {rendererClass=&NiagaraSpriteRendererProperties}
   {rendererClass=&NiagaraSpriteRendererProperties, rendererIndex=1}
  ]
 }
 NE_Star01 {
  emitterUpdateScript [
   &EmitterState(Life Cycle Mode=::NewEnumerator1, Loop Behavior=::NewEnumerator1, Loop Duration=1, UseLoopDelay=false)
   &SpawnRate(SpawnRate=15, Use Spawn Probability=false)
   &SpawnRate(SpawnGroup=1, SpawnRate=10, Use Spawn Probability=false) {module=SpawnRate001}
  ]
  particleSpawnScript [
   /SetVariables(Particles.VisibilityTag=&SpawnGroup()) {isSetParameters=true}
   &InitializeParticle(
    Color=&MergeHelper_ColorSwitch(
     0=<rgba 1 0.8164 0.2656 5>
     1=&RandomRangeLinearColor(Minimum=<rgba 1 0.2172 0.1512 0.7217>, Maximum=<rgba 1 0.3526 0.0481 0.6261>)
     2=?
     3=?
     Index=&SpawnGroup()
     Num=::NewEnumerator0
    )
    Lifetime Max=&MergeHelper_FloatSwitch(0=1.5, 1=1.2, 2=?, 3=?, Index=&SpawnGroup(), Num=::NewEnumerator0)
    Lifetime Min:NiagaraFloat=&MergeHelper_FloatSwitch(0=0.5, 1=0.8, 2=?, 3=?, Index=&SpawnGroup(), Num=::NewEnumerator0)
    Lifetime Mode=::NewEnumerator1
    Position Offset Coordinate Space=::Local
    Sprite Size Mode=::NewEnumerator1
    Uniform Sprite Size Max=&MergeHelper_FloatSwitch(0=18, 1=15, 2=?, 3=?, Index=&SpawnGroup(), Num=::NewEnumerator0)
    Uniform Sprite Size Min=&MergeHelper_FloatSwitch(0=5, 1=5, 2=?, 3=?, Index=&SpawnGroup(), Num=::NewEnumerator0)
    UsePositionOffset=true
    Write Material Random=true
   )
   &ShapeLocation(
    Box Size=<20 0 20>
    MaskSpawnGroup=false
    Override Seed=false
    Shape Primitive=::NewEnumerator2
    Surface Expansion Mode:ENiagaraExpansionMode=::NewEnumerator2
    Surface Only=true
    Surface Only Band Thickness=20
    UseOwnerScale=true
   ) {module=ShapeLocation003}
  ]
  particleUpdateScript [
   &ParticleState()
   &Drag(
    Drag=&RandomRangeFloat(
     Minimum=&MergeHelper_FloatSwitch(0=8, 1=8, 2=?, 3=?, Index=&SpawnGroup(), Num=::NewEnumerator0)
     Maximum=&MergeHelper_FloatSwitch(0=10, 1=12, 2=?, 3=?, Index=&SpawnGroup(), Num=::NewEnumerator0)
    )
    Use Linear Drag=true
    Use Rotational Drag=false
   )
   &GravityForce(Gravity=&RandomRangeVector(Minimum=<0 0 50>, Maximum=<0 0 100>))
   &ScaleColor(
    Scale Alpha=&MergeHelper_FloatSwitch(
     0=&FloatFromCurve(FloatCurve=DataInterface() {propertyValues {Curve {keys [<at -0.0092 0.0144>,<at 0.0967 0.9767>,<at 1.0011 0.002>]}}})
     1=&FloatFromCurve(FloatCurve=DataInterface() {propertyValues {Curve {keys [<at 0 0>,<at 0.2359 1>,<at 0.8018 0>]}}})
     2=?
     3=?
     Index=&SpawnGroup()
     Num=::NewEnumerator0
    )
    ScaleA=true
    ScaleRGB=true
   ) {module=ScaleColor001}
   &ScaleSpriteSize(Scale Factor=&Vector2DFromFloat(Value=&MergeHelper_FloatSwitch(
      0=&FloatFromCurve(FloatCurve=DataInterface() {propertyValues {Curve {keys [<at -0.0022 0.051>,<at 0.1 1>,<at 1 0>]}}})
      1=&FloatFromCurve(FloatCurve=DataInterface() {propertyValues {Curve {defaultValue=0.4558, keys [{tangentMode=RCTM_User, time=0, value=0.8}, {tangentMode=RCTM_User, time=1, value=1}]}}})
      2=?
      3=?
      Index=&SpawnGroup()
      Num=::NewEnumerator0
     ))) {module=ScaleSpriteSize001}
   &SolveForcesAndVelocity(Clamp Velocity=false, Limit Acceleration=false)
  ]
  rendererClasses [&NiagaraSpriteRendererProperties]
  renderers [
   {rendererClass=&NiagaraSpriteRendererProperties}
   {rendererClass=&NiagaraSpriteRendererProperties, rendererIndex=1}
  ]
 }
}
summary {
 systemName=NS_Emote_QuirkLuster_Good
 emitters [
  {emitterName=NE_Glow01, rendererClasses [&NiagaraSpriteRendererProperties]}
  {emitterName=NE_Star01, rendererClasses [&NiagaraSpriteRendererProperties]}
 ]
 userVariables [
  User.BaseCurveColor {
   type=NiagaraDataInterfaceColorCurve
   description=""
   defaultValue {
    struct=NiagaraExt_VariableValue_DataInterface
    value {
     dataInterface=&self:NiagaraDataInterfaceColorCurve_0
     dataInterfaceClass=&NiagaraDataInterfaceColorCurve
    }
   }
  }
  User.BaseCurveColorHueShift {
   type=NiagaraDataInterfaceCurve
   description=""
   defaultValue {
    struct=NiagaraExt_VariableValue_DataInterface
    value {
     dataInterface=&self:NiagaraDataInterfaceCurve_0
     dataInterfaceClass=&NiagaraDataInterfaceCurve
    }
   }
  }
  User.BaseCurveColorPick {type=NiagaraFloat, defaultValue=0, description=""}
 ]
}
